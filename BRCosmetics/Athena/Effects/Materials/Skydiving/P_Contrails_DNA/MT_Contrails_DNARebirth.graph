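# Translucent, two-sided unlit Niagara mesh-particle material for the DNA Rebirth skydiving contrail, drawing a spinning, glowing double helix.
# Emissive combines a height-based two-color gradient, noise-distorted strands, scrolling dots and a looping vertical pulse.
# Opacity combines strands, dots and rungs with edge masks, Fresnel, particle alpha and quality-switched depth and camera fades.
# WPO spins the mesh around Z with per-particle offsets, adding a sine sway that grows above the actor and cuts out at distance.
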Material /BRCosmetics/Athena/Effects/Materials/Skydiving/P_Contrails_DNA/MT_Contrails_DNARebirth
graph {
 properties {
  bUsedWithNiagaraMeshParticles=true
  blendMode=BLEND_Translucent
  pixelDepthOffsetMode=PDOM_Legacy
  shadingModel=MSM_Unlit
  twoSided=true
  usedShadingModels=Unlit
 }
 nodes [
  n1 = ObjectPositionWS()
  n2 = ObjectBounds()
  n3 = Subtract(A=ComponentMask(None=WorldPosition().XYZ) {b=true}, B=Subtract(A=ComponentMask(None=@n1) {b=true}, B=ComponentMask(None=@n2) {b=true}))
  n4 = ScalarParameter() {defaultValue=200, parameterName=ObjectSize}
  n5 = Multiply(A=Saturate(None=Divide(A=@n3, B=@n4)), B=ScalarParameter() {defaultValue=2, parameterName=ColorRampMult})
  n6 = ScalarParameter() {defaultValue=1, parameterName=ColorRampExp}
  n7 = LinearInterpolate(
   A=VectorParameter() {parameterName=ColorB, defaultValue=<rgba 2 0.375 0.8471 1>}.RGB
   B=VectorParameter() {parameterName=Color, defaultValue=<rgba 0.3681 0.1667 1 1>}.RGB
   Alpha=Saturate(None=Power(Base=@n5, Exp=@n6))
  )
  n8 = ScalarParameter() {defaultValue=1, parameterName=NoiseTileX}
  n9 = ScalarParameter() {defaultValue=5, parameterName=NoiseTileY}
  n10 = AppendVector(A=Constant(), B=DynamicParameter() {paramNames [RandOffset,Param2,Param3,Param4]}.RandOffset)
  n11 = Add(A=TextureCoordinate(), B=@n10)
  n12 = Panner(
   Coordinate=Multiply(A=AppendVector(A=@n8, B=@n9), B=@n11)
   Time=Add(A=Time()) {constB=0.3254}
   Speed=AppendVector(A=Constant(), B=ScalarParameter() {defaultValue=-1, parameterName=NoiseSpeed})
  ) {speedY=0.5}
  n13 = Add(A=@n10, B=TextureCoordinate())
  n14 = AppendVector(A=ScalarParameter() {defaultValue=0.3, parameterName=DistortionTileX}, B=ScalarParameter() {defaultValue=5, parameterName=DistortionTileY})
  n15 = Constant()
  n16 = ScalarParameter() {defaultValue=-0.3, parameterName=DistortionSpeed}
  n17 = TextureSample(UVs=Panner(Coordinate=Multiply(A=@n13, B=@n14), Speed=AppendVector(A=@n15, B=@n16))) {samplerType=SAMPLERTYPE_Grayscale, texture=&LowResBlurredNoise}
  n18 = ScalarParameter() {defaultValue=0.3, parameterName=DistortionAmnt}
  n19 = TextureSample(UVs=Add(A=@n11, B=Multiply(A=@n17.G, B=@n18))) {texture=/BRCosmetics/Athena/Effects/Textures/Skydive/P_Contrail_DNA/T_DNAHelix_strands}
  n20 = Multiply(
   A=Multiply(
    A=TextureSampleParameter2D(UVs=@n12) {parameterName=InternalNoise, texture=&T_Inky_Smoke_Tile}.G
    B=ScalarParameter() {defaultValue=0.2, parameterName=NoiseMult}
   )
   B=@n19.G
  )
  n21 = Multiply(A=AppendVector(A=Constant() {r=1}, B=ScalarParameter() {defaultValue=3, parameterName=DotsTileA}), B=@n11)
  n22 = AppendVector(A=Constant(), B=ScalarParameter() {defaultValue=-0.5, parameterName=DotsSpeedA})
  n23 = Multiply(
   A=TextureSampleParameter2D(UVs=Panner(Coordinate=@n21, Speed=@n22)) {parameterName=Dots, texture=&T_DNAHelix_strands}.R
   B=@n19.G
  )
  n24 = Add(
   A=Multiply(A=ScalarParameter() {defaultValue=200, parameterName=DotsColorMult}, B=@n23)
   B=Multiply(A=@n19.G, B=ScalarParameter() {defaultValue=1, parameterName=StrandColorMult})
  )
  n25 = Desaturation(None=Multiply(A=@n7, B=Add(A=@n20, B=@n24)), Fraction=ScalarParameter() {parameterName=StrandSaturation})
  n26 = Time() {bOverride_Period=true, period=1.5}
  n27 = ComponentMask(None=Subtract(A=ActorPositionWS(), B=WorldPosition().XYZ)) {b=true}
  n28 = Add(
   A=LinearInterpolate(
    A=ScalarParameter() {defaultValue=-300, parameterName=PulseStart}
    B=ScalarParameter() {defaultValue=600, parameterName=PulseEnd}
    Alpha=Divide(A=@n26) {constB=1.5}
   )
   B=@n27
  )
  n29 = ScalarParameter() {defaultValue=100, parameterName=PulseSize}
  n30 = Saturate(None=Abs(None=Divide(A=@n28, B=@n29)))
  n31 = Saturate(None=Power(Base=OneMinus(None=@n30)) {constExponent=10})
  n32 = VectorParameter() {parameterName=PulseColor, defaultValue=<rgba 1.0833 5.1656 16 0>}
  MaterialFunctionCall(Emissive=Add(A=Add(A=@n7, B=@n25), B=Multiply(A=@n31, B=@n32.RGB))) {materialFunction=&MF_FX_EmissiveAdjustments}
  n33 = Power(
   Base=MaterialFunctionCall(Distance=ScalarParameter() {defaultValue=90, parameterName=NearFade}) {materialFunction=&Near_Camera_Fade}.Result
   Exp=ScalarParameter() {defaultValue=5, parameterName=NearFadeExp}
  )
  n34 = Constant() {r=1}
  n35 = ShadingPathSwitch(Default=DepthFade(Opacity=Saturate(None=@n33)) {fadeDistanceDefault=18}, Mobile=@n34) {desc=Migrated from FeatureLevelSwitch}
  n36 = VertexColor()
  n37 = Add(
   A=Multiply(A=@n23, B=ScalarParameter() {defaultValue=20, parameterName=DotsOpacMult})
   B=Multiply(A=@n19.G, B=ScalarParameter() {defaultValue=3, parameterName=StrandsOpacMult})
  )
  n38 = TextureSample(UVs=TextureCoordinate() {coordinateIndex=1}) {texture=&T_DNAHelix_strands}
  n39 = Add(
   A=ScalarParameter() {defaultValue=0.15, parameterName=BaseAdd}
   B=Multiply(A=ScalarParameter() {defaultValue=0.08, parameterName=RungsAdd}, B=@n38.B)
  )
  n40 = ScalarParameter() {defaultValue=2, parameterName=OuterEdgeMult}
  n41 = ScalarParameter() {defaultValue=3, parameterName=OuterEdgeExp}
  n42 = MaterialFunctionCall!(
   UVs=TextureCoordinate() {coordinateIndex=1}
   Mult=ScalarParameter() {defaultValue=2, parameterName=RungsEdgeMult}
   Exp=ScalarParameter() {defaultValue=1, parameterName=RungsEdgeExp}
  ) {materialFunction=/Game/Effects/Fort_Effects/MaterialFunctions/MF_SquareEdgeMask}
  n43 = Multiply(
   A=Multiply(A=Power(Base=@n36.R) {constExponent=0.7}, B=Add(A=@n37, B=@n39))
   B=LinearInterpolate(
    A=MaterialFunctionCall(Mult=@n40, Exp=@n41) {materialFunction=&MF_SquareEdgeMask}.Vertmask
    B=Multiply(A=@n42.HorizMask)
    Alpha=@n38.B
   )
  )
  n44 = Constant3Vector() {constant=<rgba 0 0 1 0>}
  n45 = Time()
  n46 = DynamicParameter() {paramNames [RandOffset,Param2,Param3,Param4]}
  n47 = Fmod(A=Multiply(A=Add(A=@n45, B=@n46.RandOffset), B=ScalarParameter() {defaultValue=-0.6, parameterName=SpinSPeed}), B=Constant() {r=1})
  n48 = MaterialFunctionCall(
   Brightness=ScalarParameter() {defaultValue=1.3, parameterName=FresnelMult}
   Exponent=ScalarParameter() {defaultValue=1.5, parameterName=FresnelExp}
   Vertex / Pixel Normal=MaterialFunctionCall(Rotation Axis=@n44, Rotation Angle=@n47) {materialFunction=&FixRotateAboutAxisNormals}.New World Space Vertex Normal
  ) {materialFunction=&MF_Fresnel_FX}
  Multiply(
   A=QualitySwitch(Default=@n35, Low=@n34, High=@n35, Medium=@n35)
   B=Multiply(A=Multiply(A=@n43, B=@n48.Fresnel Double Sided), B=ParticleColor().A)
  )
  n49 = ObjectPositionWS()
  n50 = WorldPosition()
  n51 = ScalarParameter() {defaultValue=150, parameterName=WPOGradientSize}
  n52 = Saturate(None=Divide(A=Add(A=@n27, B=@n51), B=@n51))
  n53 = Subtract(A=WorldPosition().XYZ, B=Multiply(A=ActorPositionWS()) {constB=0.84})
  n54 = Time()
  n55 = Multiply(A=DynamicParameter() {paramNames [RandOffset,Param2,Param3,Param4]}.RandOffset) {constB=2}
  n56 = Add(
   A=Divide(A=ComponentMask(None=@n53) {b=true}, B=ScalarParameter() {defaultValue=600, parameterName=WPOWaveLength})
   B=Multiply(A=Add(A=@n54, B=@n55), B=ScalarParameter() {defaultValue=0.6, parameterName=WPOWavePanSpeed})
  )
  n57 = AppendVector(A=Sine(None=@n56), B=Sine(None=@n56) {period=0.73})
  n58 = Constant()
  n59 = Multiply(
   A=Power(Base=OneMinus(None=@n52)) {constExponent=3}
   B=Multiply(A=AppendVector(A=@n57, B=@n58), B=ScalarParameter() {defaultValue=20, parameterName=WPOWaveMovement})
  )
  n60 = MaterialFunctionCall(Vector 3=Subtract(A=CameraPositionWS(), B=ActorPositionWS())) {materialFunction=/Engine/Functions/Engine_MaterialFunctions02/Utility/VectorLength}
  n61 = Saturate(None=Add(A=Multiply(A=@n60.V3 Length) {constB=-1}) {constB=700})
  Reroute(None=Add(A=RotateAboutAxis(NormalizedRotationAxis=@n44, RotationAngle=@n47, PivotPoint=@n49, Position=@n50.XYZ), B=Multiply(A=@n59, B=@n61)))
 ]
}
